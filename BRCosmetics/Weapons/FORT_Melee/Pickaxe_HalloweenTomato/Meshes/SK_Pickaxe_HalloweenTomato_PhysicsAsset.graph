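# Physics asset for the HalloweenTomato pickaxe skeletal mesh, giving physics bodies to its shaft and its left and right pizza cutter parts.
# A long thin capsule on b_AR_Root covers the shaft, and each dyn_pizzaCutter bone has a zero-length capsule that acts as a sphere.
# Each cutter is constrained to the root with swing1 locked and twist free, so it can spin about its twist axis.
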PhysicsAsset /BRCosmetics/Weapons/FORT_Melee/Pickaxe_HalloweenTomato/Meshes/SK_Pickaxe_HalloweenTomato_PhysicsAsset
bodies {
 b_AR_Root [{shapeType=Capsule, length=91.755, radius=3.005, center=<7.2275 -0.0016 -23.8629>}]
 dyn_pizzaCutter_l [{shapeType=Capsule, length=0, radius=15, center=<0 -0.1 0>}]
 dyn_pizzaCutter_r [{shapeType=Capsule, length=0, radius=15, center=<0 0.1 0>}]
}
constraints [
 {
  bone1Name=dyn_pizzaCutter_l
  bone2Name=b_AR_Root
  swing1LimitDegrees=0
  swing1Motion=Locked
  swing2LimitDegrees=10
  twistLimitDegrees=45
  twistMotion=Free
 }
 {
  bone1Name=dyn_pizzaCutter_r
  bone2Name=b_AR_Root
  swing1LimitDegrees=0
  swing1Motion=Locked
  swing2LimitDegrees=10
  twistLimitDegrees=45
  twistMotion=Free
 }
]
